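; Parent quest prop actor for Security Beacons in a Season 24 narrative quest, tied to a signal jammer gameplay effect.
; Construction requires players to carry the jammer's granted instance tag.
; Interaction is allowed only for players who hold the quest and have not completed this prop's objective.
; On quest success, it removes active effects granting that tag from the player's pawn through its ability system component.
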(event OnQuestComponentSuccessEvent(ConsolidatedQuestComponent)
  (params (PlayerController :type Fort%20Player%20Controller%20Object%20Reference))
  (GameplayEffects|RemoveActiveEffectswithGrantedTags (Ability|GetFortAbilitySystemComponent (Game|Player|GetPlayerPawn PlayerController)) (GameplayTags|MakeGameplayTagContainerfromTag (Variables|InstanceGE|GetSignalJammerGrantedInstanceTag))))

(fn InstigatorCanInteract
  (params (InstigatingPlayerController :type Fort%20Player%20Controller%20Object%20Reference))
  (bind _consolidatedquestcomponent (Variables|BuildingGameplayActorPropQuest|GetConsolidatedQuestComponent))
  (bind _returnvalue (Class|ConsolidatedQuestComponent|GetQuestItemDefinition _consolidatedquestcomponent))
  ;[0 0 0] Check if this player is on the quest and they have not already completed the objective for this prop instance
  (Utilities|IsValid InstigatingPlayerController
    (:Is%20Valid
      (Utilities|IsValid _returnvalue
        (:Is%20Valid
          (Utilities|IsValid (Quests|GetQuestManager InstigatingPlayerController "Invalid")
            (:Is%20Valid
              (bind _playerhasquest (Class|AthenaQuestFunctionLibrary|AthenaHasQuest InstigatingPlayerController _returnvalue))
              (if _playerhasquest
                (bind (_playercontrollerout _completedquest) (Quest|HasPlayerCompletedQuest InstigatingPlayerController _returnvalue (Variables|/Script/FortniteGame.ConsolidatedQuestComponent|GetObjectiveBackendName (Variables|BuildingGameplayActorPropQuest|GetConsolidatedQuestComponent))))
                (if (not _completedquest)
                  (return true)
                  (else
                    _))
                (else
                  _))
              (return false))
            (:Is%20Not%20Valid
              _)))
        (:Is%20Not%20Valid
          _)))
    (:Is%20Not%20Valid
      _)))

(fn ConstructionScript
  (|CallParentFunction)
  (Variables|Design|QuestSetup|GameplayTags,EffectsandCosmetics|SetRequired%20Player%20Tags (GameplayTags|MakeGameplayTagContainerfromTag (Variables|InstanceGE|GetSignalJammerGrantedInstanceTag))))
